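# Translucent unlit material for animated Slurp fluid on Pickaxe_18, usable on skeletal and Niagara meshes.
# A plasma texture, UV-distorted by a water-ripple normal map, pans along UV1, and its contrast-adjusted value blends two Slurp colors.
# A dot texture panning the opposite way darkens that intensity-scaled color as bubbles before MF_FX_EmissiveAdjustments.
# The same dots, boosted and inverted, also form a separate mask.
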Material /BRCosmetics/Weapons/FORT_Melee/Pickaxe_18/Materials/M_Pickaxe_SlurpFluid
graph {
 properties {
  bUsedWithMeshDeformer=true
  bUsedWithNiagaraMeshParticles=true
  bUsedWithSkeletalMesh=true
  blendMode=BLEND_Translucent
  pixelDepthOffsetMode=PDOM_Legacy
  shadingModel=MSM_Unlit
  usedShadingModels=Unlit
 }
 nodes [
  n1 = TextureCoordinate() {coordinateIndex=1}
  n2 = ComponentMask(None=VectorParameter() {parameterName=Bubbles_UV_Scale, defaultValue=<rgba 0.5 0.25 0 0>}.RGB) {g=true, r=true}
  n3 = Multiply(A=Time(), B=ScalarParameter() {defaultValue=5, parameterName=Time Multi})
  n4 = TextureSample(UVs=Panner(Coordinate=Multiply(A=@n1, B=@n2), Time=@n3) {speedX=0.0625}) {texture=/Game/Effects/Fort_Effects/Textures/Tile/T_LaserTrail_Dots}
  n5 = ScalarParameter() {defaultValue=20}
  LinearInterpolate(Alpha=Clamp(None=Multiply(A=@n4.R, B=@n5))) {constA=1, constB=0}
  n6 = TextureCoordinate() {coordinateIndex=1}
  n7 = Multiply(A=@n6, B=ComponentMask(None=VectorParameter() {parameterName=Liquid_UV_Scale, defaultValue=<rgba 8 2 0 0>}.RGB) {g=true, r=true})
  n8 = ComponentMask(None=TextureSample(UVs=@n6) {samplerType=SAMPLERTYPE_Normal, texture=/Game/Effects/Fort_Effects/Gameplay/Water/Textures/T_WaterRipples_01}.RGB) {g=true, r=true}
  n9 = TextureSample(UVs=Panner(Coordinate=Add(A=@n7, B=@n8), Time=@n3) {speedX=-0.5}) {texture=/Game/Effects/Fort_Effects/Textures/Tile/T_Plasma_BallCore}
  n10 = ScalarParameter() {defaultValue=1.25, parameterName=Contrast}
  n11 = LinearInterpolate(
   A=VectorParameter() {parameterName=Slurp Color B, defaultValue=<rgba 0.0019 0.0606 0.125 1>}.RGB
   B=VectorParameter() {parameterName=Slurp Color A, defaultValue=<rgba 0.0153 2 0.8084 1>}.RGB
   Alpha=Clamp(None=Power(Base=@n9.R, Exp=@n10))
  )
  n12 = ScalarParameter() {defaultValue=1, parameterName=Intensity}
  MaterialFunctionCall(Emissive=LinearInterpolate(A=Multiply(A=@n11, B=@n12), Alpha=Clamp(None=@n4.R)) {constA=1, constB=0}) {materialFunction=&MF_FX_EmissiveAdjustments}
 ]
}
