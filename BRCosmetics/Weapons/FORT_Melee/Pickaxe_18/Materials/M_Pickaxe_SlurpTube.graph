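# M_Pickaxe_SlurpTube is a translucent Default Lit material, evidently for the slurp tube on the Pickaxe_18 melee weapon.
# It uses surface translucency lighting and index-of-refraction refraction, and supports skeletal meshes and mesh deformers.
# The graph holds only constants and parameters for tube color, clear coat color, metallic, opacity and refraction, with no texture samples and no connections shown.
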Material /BRCosmetics/Weapons/FORT_Melee/Pickaxe_18/Materials/M_Pickaxe_SlurpTube
graph {
 properties {
  bUsedWithMeshDeformer=true
  bUsedWithSkeletalMesh=true
  blendMode=BLEND_Translucent
  pixelDepthOffsetMode=PDOM_Legacy
  refractionMethod=RM_IndexOfRefraction
  translucencyLightingMode=TLM_Surface
  usedShadingModels=Default Lit
 }
 nodes [
  Constant3Vector() {constant=<rgba 0 0 1 0>}
  Constant()
  Constant()
  Constant() {r=1}
  ScalarParameter() {defaultValue=1, parameterName=Metallic}
  ScalarParameter() {defaultValue=0.5, parameterName=Opacity}
  ScalarParameter() {defaultValue=1, parameterName=refraction}
  VectorParameter() {parameterName=Tube Color, defaultValue=<rgba 0.6302 0.6302 0.6302 1>}
  VectorParameter() {parameterName=Clear_Coat_Color, defaultValue=<rgba 1 1 1 1>}
 ]
}
