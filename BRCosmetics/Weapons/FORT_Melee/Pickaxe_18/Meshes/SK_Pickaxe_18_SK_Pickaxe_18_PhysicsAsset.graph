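# Physics asset for the Pickaxe_18 two-handed melee pickaxe skeletal mesh in Battle Royale cosmetics.
# A single box body on the Melee_TwoHanded_Handle bone encloses the weapon in one slightly offset and rotated shape.
# It has no constraints, so the pickaxe is simulated or collided as one rigid piece.
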PhysicsAsset /BRCosmetics/Weapons/FORT_Melee/Pickaxe_18/Meshes/SK_Pickaxe_18_SK_Pickaxe_18_PhysicsAsset
bodies {Melee_TwoHanded_Handle [{shapeType=Box, extentX=47.6307, extentY=36.9172, extentZ=158.5685, center=<2.1078 1.0598 30.2638>, rotation=<rot -0.5411 -90 -0.9519>}]}
constraints []
